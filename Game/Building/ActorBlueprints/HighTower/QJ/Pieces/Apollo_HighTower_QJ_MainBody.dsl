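; Main body piece of the HighTower QJ building actor, managing its lobby and interior idle audio.
; Start Lobby Audio spawns a given sound at the actor's location, stores the component and fades it in.
; Set Lobby Audio Parameter drives that loop's RampUp float parameter.
; On death, both the lobby loop and the interior idle loop fade out, each guarded by a validity check.
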(event Custom|Start%20Lobby%20Audio
  (params (Sound :type Sound%20Base%20Object%20Reference))
  (bind _returnvalue (Audio|SpawnSoundatLocation Sound (Transformation|GetActorLocation)))
  (bind _output_get (Variables|Default|SetLobbyIdleAudioLoop _returnvalue))
  (Audio|Components|Audio|FadeIn _output_get 3.0 1.0 0.0 "Linear" :!class /Script/Engine.AudioComponent))

(event Custom|Set%20Lobby%20Audio%20Parameter
  (params (InFloat :type Float%20%28double-precision%29))
  (bind _lobbyidleaudioloop (Variables|Default|GetLobbyIdleAudioLoop)
    (:then
      (Audio|Parameter|SetFloatParameter _lobbyidleaudioloop "RampUp" InFloat :!class /Script/AudioExtensions.AudioParameterControllerInterface))
    (:else)))

(event EventOnDeathPlayEffects
  (params (Damage :type Float%20%28single-precision%29) (DamageTags :type Gameplay%20Tag%20Container%20Structure%20%28by%20ref%29) (Momentum :type Vector :default "0, 0, 0") (HitInfo :type Hit%20Result%20Structure%20%28by%20ref%29) (InstigatedBy :type Fort%20Pawn%20Object%20Reference) (DamageCauser :type Actor%20Object%20Reference) (EffectContext :type Gameplay%20Effect%20Context%20Handle%20Structure))
  (Utilities|FlowControl|Sequence
    (:then_0
      (bind _lobbyidleaudioloop (Variables|Default|GetLobbyIdleAudioLoop)
        (:then
          (Audio|Components|Audio|FadeOut _lobbyidleaudioloop 0.5 0.0 "Linear" :!class /Script/Engine.AudioComponent))
        (:else)))
    (:then_1
      (bind _veh_hightower_qj_idle_interior_loop_01 (Variables|Default|GetVeh_Hightower_QJ_Idle_Interior_Loop_01)
        (:then
          (Audio|Components|Audio|FadeOut _veh_hightower_qj_idle_interior_loop_01 0.5 0.0 "Linear" :!class /Script/Engine.AudioComponent))
        (:else)))))

(fn ConstructionScript)
